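# Niagara effect for the BuffCatComic emote that emits short-lived mesh particles, likely line shapes given its name.
# A single emitter spawns at a steady rate, lifting the spawn point vertically with looped emitter age to build a rising stack.
# Meshes get randomized orientation and a scale range tied to emitter age, then rotate about Z and grow.
# A curve sweeps a dynamic material parameter from negative to positive, animating the material.
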NiagaraSystem /Game/Effects/Fort_Effects/Emotes/BuffCatComic/NS_Emote_Line
dependencies {usedDataInterfaces [&NiagaraDataInterfaceCurve], usedDynamicInputs [], usedModules [&SystemState]}
emitters {Empty {
  emitterUpdateScript [
   &EmitterState(Life Cycle Mode=::NewEnumerator1, Loop Behavior=::NewEnumerator1, Loop Duration=0.45, UseLoopDelay=false)
   &SpawnRate(SpawnRate=39, Use Spawn Probability=false)
  ]
  particleSpawnScript [
   &InitializeParticle(
    Color Mode=::NewEnumerator0
    Lifetime Max=0.7
    Lifetime Min=0.4
    Lifetime Mode=::NewEnumerator1
    Mesh Scale Max=<1.5 1.5 1>
    Mesh Scale Min=&Multiply_Vector(
     A=&RandomRangeVector(Minimum=<1 1 1>, Maximum=<1.2 1.2 1.2>)
     B=&VectorFromFloat(Value=&FloatFromCurve(FloatCurve=DataInterface() {propertyValues {Curve {keys [<at 0 0.7>,<at 1 1.5>]}}}, CurveIndex=$Emitter.Age))
    )
    Mesh Scale Mode=::NewEnumerator2
    UsePositionOffset=false
    Write Material Random=true
   )
   &InitialMeshOrientation(
    Rotation=&RandomRangeVector(Minimum=<0 0 0>, Maximum=<0 0 1>)
    Use Orientation Vector=true
    Use Rotation Vector=true
   )
   &SystemLocation(MaskSpawnGroup=false, Offset=&Lerp_Vector(A=<0 0 0>, B=<0 0 320>, Alpha=$Emitter.LoopedAge))
  ]
  particleUpdateScript [
   &ParticleState()
   &UpdateMeshOrientation(Rotation Vector=&RandomRangeVector(Minimum=<0 0 -0.5>, Maximum=<0 0 -1.2>))
   &ScaleMeshSize(Scale Factor=&VectorFromFloat(Value=&FloatFromCurve(FloatCurve=DataInterface() {propertyValues {Curve {keys [<at 0 1>,<at 1 1.2>]}}})))
   &DynamicMaterialParameters(
    Index 0 Param 1=&FloatFromCurve(FloatCurve=DataInterface() {propertyValues {Curve {keys [<at 0 -1>,<at 1 1>]}}})
    Module.Param0WriteEnabled=true
    Module.Param1WriteEnabled=true
    Module.Param2WriteEnabled=true
    Module.Param3WriteEnabled=true
   )
  ]
  rendererClasses [&NiagaraMeshRendererProperties]
  renderers [{rendererClass=&NiagaraMeshRendererProperties, rendererIndex=0}]
 }}
summary {systemName=NS_Emote_Line, emitters [{emitterName=Empty, rendererClasses [&NiagaraMeshRendererProperties]}], userVariables []}
